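# UI-domain translucent material for the Apollo minimap base map in Athena.
# It consists of a single MF_Minimap material function call, which carries the minimap rendering logic.
# The call exposes UVsOverride and TextureOverride inputs, with no connections to them shown in the graph.
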Material /Game/Athena/Apollo/Maps/UI/M_MiniMap_Apollo_MapBase1
graph {
 properties {blendMode=BLEND_Translucent, materialDomain=MD_UI, pixelDepthOffsetMode=PDOM_Legacy, usedShadingModels=Default Lit}
 nodes [MaterialFunctionCall!(UVsOverride, TextureOverride) {materialFunction=&/MF_Minimap}]
}
