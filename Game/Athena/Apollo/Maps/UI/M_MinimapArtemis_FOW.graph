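# UI material for the Artemis minimap that blends the map with a fog-of-war version, including an animated discovery reveal.
# RT_FogofWar's red channel drives the blend directly, while green gates a radial wipe centered on DiscoverabilityMPC's CurrentStartLocation and grown by DiscoverRevealProgress.
# Panning Perlin noise distorts the MF_Minimap_FOW UVs along the wipe's soft edge.
# An offset LineColor tint is added within the green mask.
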Material /Game/Athena/Apollo/Maps/UI/M_MinimapArtemis_FOW
graph {
 properties {blendMode=BLEND_Translucent, materialDomain=MD_UI, pixelDepthOffsetMode=PDOM_Legacy, usedShadingModels=Default Lit}
 nodes [
  n1 = TextureSample() {texture=&/RT_FogofWar}
  n2 = CollectionParameter() {collection=&/DiscoverabilityMPC, parameterName=CurrentStartLocation}
  n3 = ScalarParameter() {defaultValue=0.04, parameterName=SIze}
  n4 = ScalarParameter() {defaultValue=0.6, parameterName=OffsetEnd}
  n5 = Subtract(
   A=Distance(A=TextureCoordinate(), B=ComponentMask(None=@n2) {g=true, r=true})
   B=LinearInterpolate(
    A=Multiply(A=@n3) {constB=-1}
    B=@n4
    Alpha=CollectionParameter() {collection=&DiscoverabilityMPC, parameterName=DiscoverRevealProgress}
   )
  )
  n6 = Saturate(None=Divide(A=@n5, B=@n3))
  n7 = Saturate(None=Add(A=@n1.R, B=Multiply(A=@n1.G, B=@n6)))
  n8 = TextureCoordinate()
  n9 = ScalarParameter() {defaultValue=6, parameterName=DistortScale2}
  n10 = TextureSample(UVs=Panner(Coordinate=Multiply(A=@n8, B=@n9), Speed=ScalarParameter() {defaultValue=0.25, parameterName=DistortionSpeed})) {texture=/Game/Global/Textures/Default/Noise/3xPerlin_M}
  n11 = AppendVector(A=@n10.R, B=@n10.R)
  n12 = ComponentMask(None=VectorParameter() {parameterName=Bias2, defaultValue=<rgba 0.25 0.25 0 1>}.RGB) {g=true, r=true}
  n13 = OneMinus(None=@n6)
  n14 = Multiply(A=Multiply(A=@n1.G, B=Multiply(A=@n6, B=@n13)), B=ScalarParameter() {defaultValue=0.1, parameterName=UVdistortion2})
  n15 = Add(A=Multiply(A=Subtract(A=@n11, B=@n12) {constB=0.5}, B=@n14), B=TextureCoordinate())
  n16 = VectorParameter() {parameterName=LineColor, defaultValue=<rgba 0.05 0.05 0.05 1>}
  n17 = Divide(
   A=Add(A=ScalarParameter() {defaultValue=-0.0491, parameterName=offsetLine}, B=@n5)
   B=ScalarParameter() {defaultValue=0.005, parameterName=SIzeLine}
  )
  n18 = Saturate(None=Sine(None=Saturate(None=@n17)))
  Add(
   A=LinearInterpolate(
    A=MaterialFunctionCall!(UVsOverride, TextureOverride) {materialFunction=&/MF_Minimap}.RGB
    B=MaterialFunctionCall!(UVsOverride=@n15) {materialFunction=&/MF_Minimap_FOW}.RGB
    Alpha=@n7
   )
   B=Multiply(A=@n1.G, B=Multiply(A=@n16.RGB, B=@n18))
  )
  Add(A=@n14, B=ScalarParameter() {defaultValue=0.002, parameterName=UVdistortion_Ambient})
  Constant() {r=1}
  Frac(None=Multiply(A=Time(), B=ScalarParameter() {defaultValue=0.2, parameterName=Speed}))
  n19 = VectorParameter() {desc=XY location and Z is Radius of the POI, parameterName=LocationRevealFX, defaultValue=<rgba 0.6788 0.6979 0.3 1>}
  Multiply(A=@n19.B, B=@n4)
  Multiply(A=@n7, B=MaterialFunctionCall() {materialFunction=&MF_Minimap}.Alpha)
  n20 = TextureSample(UVs=Multiply(A=TextureCoordinate(), B=ScalarParameter() {defaultValue=2, parameterName=DistortScale1})) {texture=&3xPerlin_M}
  n21 = VectorParameter() {parameterName=Bias, defaultValue=<rgba 0.25 0.25 0 1>}
  n22 = Multiply(
   A=Subtract(A=AppendVector(A=@n20.R, B=@n20.R), B=ComponentMask(None=@n21.RGB) {g=true, r=true}) {constB=0.5}
   B=ScalarParameter() {defaultValue=0.1099, parameterName=UVdistortion}
  )
  n23 = TextureCoordinate()
  n24 = Subtract(A=Distance(A=Add(A=@n22, B=@n23), B=ComponentMask(None=@n2) {g=true, r=true}), B=@n4)
  OneMinus(None=Saturate(None=Divide(A=@n24) {constB=0.001}))
  ScalarParameter() {defaultValue=-0.1, parameterName=OffsetStart}
 ]
}
