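# Translucent unlit map UI material with logic for encoding and testing an integer BitValue as an 8-bit mask value, apparently for map discoverability.
# One branch scales BitValue by 1/256 inside a centered UV SphereMask with adjustable radius and hardness.
# Another scales the Map Discoverability texture's red channel by 255 and outputs one only on equality with BitValue, rather than a bitwise test.
# No material output connections are shown.
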Material /Game/Athena/Apollo/Maps/UI/M_ViewStencil
graph {
 properties {blendMode=BLEND_Translucent, pixelDepthOffsetMode=PDOM_Legacy, shadingModel=MSM_Unlit, usedShadingModels=Unlit}
 nodes [
  Constant() {r=1}
  n1 = Constant() {r=1}
  n2 = ScalarParameter() {defaultValue=2, parameterName=BitValue}
  Multiply(
   A=Multiply(A=Divide(A=@n1) {constB=256}, B=@n2)
   B=SphereMask(
    A=TextureCoordinate()
    B=Constant2Vector() {g=0.5, r=0.5}
    Radius=ScalarParameter() {defaultValue=0.45, parameterName=Radius}
    Hardness=ScalarParameter() {defaultValue=1, parameterName=Hardness}
   )
  )
  n3 = TextureSampleParameter2D() {parameterName=Map Discoverability, texture=/Game/Athena/UI/Frontend/Spatial/Art/T_Ui_Empty}
  n4 = Constant()
  Saturate(None=If(A=Multiply(A=@n3.R) {constB=255}, B=@n2, "A > B"=@n4, "A == B"=Constant() {r=1}, "A < B"=@n4))
  ScalarParameter() {defaultValue=13}
  VectorParameter() {parameterName=ChannelMask, defaultValue=<rgba 0 0 0 1>}
 ]
}
