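# Post-process material that grades the scene for Apollo minimap screenshots, judging by its name and terrain mask.
# It applies power, brightness and saturation adjustments, then blends a water color into pixels whose shading model ID exceeds a threshold.
# A screen-space mask texture drives a tint and fades to a water vignette color, with an Enabled parameter blending against the original scene.
# Forward and mobile shading paths receive constant black instead.
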Material /Game/Athena/Apollo/Maps/UI/PPM_MinimapShot
graph {
 properties {blendableOutputAlpha=true, materialDomain=MD_PostProcess, pixelDepthOffsetMode=PDOM_Legacy, usedShadingModels=Default Lit}
 nodes [
  Constant() {r=1}
  n1 = Constant()
  If(
   A=ComponentMask(None=SceneTexture() {sceneTextureId=PPI_ShadingModelID}.Color) {r=true}
   B=ScalarParameter() {defaultValue=1, parameterName=SahdingModelWater}
   "A > B"=@n1
   "A == B"=Constant() {r=1}
   "A < B"=@n1
  )
  n2 = Subtract(
   A=ComponentMask(None=SceneTexture() {sceneTextureId=PPI_ShadingModelID}.Color) {r=true}
   B=ScalarParameter() {defaultValue=7.4323, parameterName=Threshold1}
  )
  n3 = OneMinus(None=Saturate(None=Ceil(None=@n2)))
  Reroute(None=@n3)
  n4 = TextureSample(UVs=ScreenPosition().ViewportUV) {samplerType=SAMPLERTYPE_Masks, texture=&/Apollo_Terrain_Minimap_PostMask}
  n5 = LinearInterpolate(A=Constant3Vector() {constant=<rgba 1 1 1 0>}, B=VectorParameter() {parameterName=Color2, defaultValue=<rgba 2 2 2 1>}.RGB, Alpha=@n4.R)
  n6 = ComponentMask(None=SceneTexture() {sceneTextureId=PPI_PostProcessInput0}.Color) {b=true, g=true, r=true}
  n7 = ScalarParameter() {defaultValue=1, parameterName=Power}
  n8 = Desaturation(
   None=Multiply(A=Power(Base=@n6, Exp=@n7), B=ScalarParameter() {defaultValue=1, parameterName=Brightness})
   Fraction=ScalarParameter() {defaultValue=-0.1, parameterName=Saturation}
  )
  n9 = LinearInterpolate(
   A=LinearInterpolate(
    A=VectorParameter() {parameterName=Color1Water, defaultValue=<rgba 1 1 1 1>}.RGB
    B=@n8
    Alpha=ScalarParameter() {defaultValue=0.5, parameterName=WaterBlend}
   )
   B=@n8
   Alpha=@n3
  )
  n10 = LinearInterpolate(
   A=ComponentMask(None=SceneTexture() {sceneTextureId=PPI_PostProcessInput0}.Color) {b=true, g=true, r=true}
   B=LinearInterpolate(A=VectorParameter() {parameterName=Color3WaterVignette, defaultValue=<rgba 2 2 2 1>}.RGB, B=Multiply(A=@n5, B=@n9), Alpha=@n4.R)
   Alpha=ScalarParameter() {defaultValue=1, parameterName=Enabled}
  )
  n11 = Constant3Vector()
  ShadingPathSwitch(Default=@n10, Deferred=@n10, Forward=@n11, Mobile=@n11)
 ]
}
