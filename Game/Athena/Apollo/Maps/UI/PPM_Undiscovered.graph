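# Post-process material that replaces the tonemapper to render a stylized blue image, apparently for undiscovered areas of the Apollo map UI.
# It posterizes scene brightness and uses a shading-model ID threshold to choose between an ocean tint and a two-color land gradient.
# Four offset samples of posterized scene color form darkened outlines, and an Enabled parameter blends in the result.
# Forward and mobile paths get a constant, and a static switch adds optional sRGB compensation.
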Material /Game/Athena/Apollo/Maps/UI/PPM_Undiscovered
graph {
 properties {blendableLocation=BL_ReplacingTonemapper, materialDomain=MD_PostProcess, pixelDepthOffsetMode=PDOM_Legacy, usedShadingModels=Default Lit}
 nodes [
  Constant()
  n1 = SceneTexture() {sceneTextureId=PPI_ShadingModelID}
  n2 = Constant()
  If(
   A=ComponentMask(None=@n1.Color) {r=true}
   B=ScalarParameter() {defaultValue=1, parameterName=SahdingModelWater}
   "A > B"=@n2
   "A == B"=Constant() {r=1}
   "A < B"=@n2
  )
  MaterialFunctionCall!(Distance, Distance Steps, Radial Steps, Radial Offset, TempAA Radial Blur, TempAA Distance Blur, Distance Mask) {materialFunction=/Game/Effects/Fort_Effects/Materials/_Functions/SpiralBlur-SceneTexture}
  n3 = ScalarParameter() {parameterName=ContrastOcean1}
  n4 = ScalarParameter() {defaultValue=1, parameterName=ContrastOcean2}
  n5 = ScalarParameter() {defaultValue=12, parameterName=StepsVis}
  n6 = Multiply(A=ComponentMask(None=SceneTexture() {sceneTextureId=PPI_PostProcessInput0}.Color) {b=true, g=true, r=true}, B=@n5)
  n7 = Desaturation(None=Divide(A=Floor(None=@n6), B=@n5))
  n8 = Multiply(
   A=VectorParameter() {parameterName=Blue, defaultValue=<rgba 0.2536 0.494 0.8542 1>}.RGB
   B=Saturate(None=LinearInterpolate(A=@n3, B=@n4, Alpha=@n7))
  )
  n9 = ScalarParameter() {parameterName=Contrast1}
  n10 = ScalarParameter() {defaultValue=1, parameterName=Contrast2}
  n11 = LinearInterpolate(
   A=VectorParameter() {parameterName=Color1, defaultValue=<rgba 0.0249 0.3391 1 1>}.RGB
   B=VectorParameter() {parameterName=Color2, defaultValue=<rgba 0.2779 1.0054 2 1>}.RGB
   Alpha=Saturate(None=LinearInterpolate(A=@n9, B=@n10, Alpha=@n7))
  )
  n12 = Subtract(A=ComponentMask(None=@n1.Color) {r=true}, B=ScalarParameter() {defaultValue=7.4323, parameterName=Threshold1})
  n13 = OneMinus(None=Saturate(None=Ceil(None=@n12)))
  n14 = SceneTexture() {sceneTextureId=PPI_PostProcessInput0}
  n15 = ScalarParameter() {defaultValue=12, parameterName=Steps}
  n16 = Floor(None=Multiply(A=ComponentMask(None=@n14.Color) {b=true, g=true, r=true}, B=@n15))
  n17 = Desaturation(None=LinearInterpolate(A=Constant3Vector() {constant=<rgba 1 1 1 0>}, B=Divide(A=@n16, B=@n15), Alpha=@n13))
  n18 = ScreenPosition()
  n19 = ScalarParameter() {defaultValue=0.01, parameterName=LineThickness}
  n20 = MaterialFunctionCall!(OffsetAmount=AppendVector(A=@n19, B=@n19)) {materialFunction=/Engine/Functions/Engine_MaterialFunctions01/ImageAdjustment/BlurSampleOffsets}
  n21 = ComponentMask(None=SceneTexture(UVs=Add(A=@n18.ViewportUV, B=@n20."0,1")) {sceneTextureId=PPI_PostProcessInput0}.Color) {b=true, g=true, r=true}
  n22 = ScalarParameter() {defaultValue=12, parameterName=Steps}
  n23 = Divide(A=Floor(None=Multiply(A=@n21, B=@n22)), B=@n22)
  n24 = ScreenPosition()
  n25 = ComponentMask(None=SceneTexture(UVs=Add(A=@n24.ViewportUV, B=@n20."1,0")) {sceneTextureId=PPI_PostProcessInput0}.Color) {b=true, g=true, r=true}
  n26 = ScalarParameter() {defaultValue=12, parameterName=Steps}
  n27 = Divide(A=Floor(None=Multiply(A=@n25, B=@n26)), B=@n26)
  n28 = Add(A=Subtract(A=@n17, B=Desaturation(None=@n23)), B=Subtract(A=@n17, B=Desaturation(None=@n27)))
  n29 = SceneTexture(UVs=Add(A=ScreenPosition().ViewportUV, B=@n20."0,-1")) {sceneTextureId=PPI_PostProcessInput0}
  n30 = ScalarParameter() {defaultValue=12, parameterName=Steps}
  n31 = Floor(None=Multiply(A=ComponentMask(None=@n29.Color) {b=true, g=true, r=true}, B=@n30))
  n32 = Subtract(A=@n17, B=Desaturation(None=Divide(A=@n31, B=@n30)))
  n33 = Add(A=ScreenPosition().ViewportUV, B=@n20."-1,0")
  n34 = ScalarParameter() {defaultValue=12, parameterName=Steps}
  n35 = Multiply(A=ComponentMask(None=SceneTexture(UVs=@n33) {sceneTextureId=PPI_PostProcessInput0}.Color) {b=true, g=true, r=true}, B=@n34)
  n36 = Desaturation(None=Divide(A=Floor(None=@n35), B=@n34))
  n37 = Divide(A=Add(A=Add(A=@n28, B=@n32), B=Subtract(A=@n17, B=@n36)), B=ScalarParameter() {defaultValue=3, parameterName=OutlineConstrast})
  n38 = OneMinus(None=Multiply(A=Saturate(None=@n37), B=ScalarParameter() {defaultValue=0.2, parameterName=OpacityOutline}))
  n39 = LinearInterpolate(
   A=ComponentMask(None=SceneTexture() {sceneTextureId=PPI_PostProcessInput0}.Color) {b=true, g=true, r=true}
   B=Multiply(A=LinearInterpolate(A=@n8, B=@n11, Alpha=@n13), B=Saturate(None=@n38))
   Alpha=ScalarParameter() {defaultValue=1, parameterName=Enabled}
  )
  n40 = Constant3Vector()
  n41 = ShadingPathSwitch(Default=@n39, Deferred=@n39, Forward=@n40, Mobile=@n40)
  n42 = MaterialFunctionCall(Float3=@n41) {materialFunction=&BreakOutFloat3Components}
  MaterialFunctionCall(
   X=MaterialFunctionCall!(sRGB=@n42.R) {materialFunction=/Engine/Functions/Engine_MaterialFunctions01/ImageAdjustment/sRGBToLinear}.Linear
   Y=MaterialFunctionCall(sRGB=@n42.G) {materialFunction=&sRGBToLinear}.Linear
   Z=MaterialFunctionCall(sRGB=@n42.B) {materialFunction=&sRGBToLinear}.Linear
  ) {materialFunction=&MakeFloat3}
  Multiply()
  SceneTexture() {sceneTextureId=PPI_SceneDepth}
  n43 = MaterialFunctionCall(Float3=@n41) {materialFunction=&BreakOutFloat3Components}
  StaticSwitchParameter(
   True=MaterialFunctionCall(
    X=MaterialFunctionCall!(Linear=@n43.R) {materialFunction=/Engine/Functions/Engine_MaterialFunctions01/ImageAdjustment/LinearTosRGB}.sRGB
    Y=MaterialFunctionCall(Linear=@n43.G) {materialFunction=&LinearTosRGB}.sRGB
    Z=MaterialFunctionCall(Linear=@n43.B) {materialFunction=&LinearTosRGB}.sRGB
   ) {materialFunction=&MakeFloat3}.Result
   False=@n41
  ) {parameterName=sRGBcompensationSet True for Tweaking}
 ]
}
