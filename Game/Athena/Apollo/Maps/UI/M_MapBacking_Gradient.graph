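# Translucent UI material that draws a radial gradient backing, apparently for the map screen, blending two color parameters (dark to brighter blue by default).
# GetUserInterfaceUV pixel size yields an aspect ratio that scales the UVs and center, keeping the gradient circular on non-square widgets.
# RadialGradientExponential, with parameterized center offset, radius and density, drives the color lerp, which a Darkening multiplier scales.
# A separate Alpha parameter sits alongside for overall transparency.
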Material /Game/Athena/Apollo/Maps/UI/M_MapBacking_Gradient
graph {
 properties {blendMode=BLEND_Translucent, materialDomain=MD_UI, pixelDepthOffsetMode=PDOM_Legacy, usedShadingModels=Default Lit}
 nodes [
  n1 = MaterialFunctionCall() {materialFunction=&GetUserInterfaceUV}
  n2 = MaterialFunctionCall(Float2=@n1.Pixel Size) {materialFunction=&BreakOutFloat2Components}
  n3 = AppendVector(A=Divide(A=@n2.R, B=Max(A=@n2.G)), B=Constant() {r=1})
  n4 = Multiply(A=@n3, B=@n1.Normalized UV)
  n5 = Add(A=Multiply(A=@n3) {constB=0.5}, B=AppendVector(A=ScalarParameter(), B=ScalarParameter() {parameterName=Param_1}))
  n6 = ScalarParameter() {defaultValue=0.95, parameterName=Radius}
  n7 = ScalarParameter() {defaultValue=2, parameterName=Density}
  Multiply(
   A=LinearInterpolate(
    A=VectorParameter() {parameterName=RGB1, defaultValue=<rgba 0 0.006 0.1095 1>}.RGB
    B=VectorParameter() {parameterName=RGB2, defaultValue=<rgba 0 0.1384 0.3763 1>}.RGB
    Alpha=MaterialFunctionCall(UVs=@n4, CenterPosition=@n5, Radius=@n6, Density=@n7) {materialFunction=&RadialGradientExponential}.RadialGradientExponential
   )
   B=ScalarParameter() {defaultValue=0.65, parameterName=Darkening}
  )
  ScalarParameter() {defaultValue=0.925, parameterName=Alpha}
  TextureCoordinate()
 ]
}
